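# Control Rig for the F_MED_MasalaRail female character body, adding corrective deformation and leg-driven belt and cloth adjustment.
# A sequence runs the shared DeformRig function, then spherical pose readers on both thighs, which have debug drawing enabled.
# Each reader weights offsets on its side's belt and cloth, while the larger value drives shared cloth and front-belt offsets.
# Weighted parent constraints place the metal belt pieces before pose-driven rotations.
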ControlRigBlueprint /BRCosmetics/Characters/Player/Female/Medium/Bodies/F_MED_MasalaRail/Meshes/F_MED_MasalaRail_CtrlRig
graphs [
 RigVMFunctionLibrary {links [], nodes []}
 RigVMModel {nodes [
   RigUnit_BeginExecution = &RigUnit_BeginExecution()
   RigVMFunction_Sequence = &RigVMFunction_Sequence(ExecuteContext=@RigUnit_BeginExecution.ExecutePin)
   FN_DeformRig_FNC = RigVMFunctionReferenceNode(
    DeformRig_ Curve="(deformRig_l_upperarm_up_alpha_72_#1=False,deformRig_r_upperarm_up_alpha_73_#2=False,deformRig_l_upperarm_fr_alpha_74_#3=False,deformRig_r_upperarm_fr_alpha_75_CB51B8 ...(truncated)"
    DeformRig_Bool="(upperarm_1_#4=True,elbow_3_#5=True,wrist_7_#6=True,groin_8_#7=True,glute_17_#8=True,knee_9_40C408324CA261 ...(truncated)"
    DeformRig_Debug=(upperarm_debug_18_#4=False,wrist_debug_20_#6=False,lat_debug_26_#9=False,groinGlute_debug_27_#7=False)
    DeformRig_Value="(upperarm_up_18_#10=(Rotation=(X=0,Y=0,Z=0,W=1),Translation=(X=0,Y=0,Z=8),Scale3D=(X=0.7,Y=1,Z=1)),upperarm_fwd_fr_19_#11=(Rotation=(X=0,Y=0,Z=0,W=1),Translation=(X=0,Y=10,Z=0),Scale3D=(X=0.5,Y=1,Z= ...(truncated)"
    ExecuteContext=@RigVMFunction_Sequence.A
   )
   n1 = &RigUnit_SphericalPoseReader(
    ActiveRegionScaleFactors=(PositiveWidth=1,NegativeWidth=0,PositiveHeight=1,NegativeHeight=1)
    ActiveRegionSize=0.3
    Debug=(bDrawDebug=true,bDraw2D=false,bDrawLocalAxes=false,DebugScale=30,DebugSegments=30,DebugThickness=0.2)
    DriverAxis=(X=-1,Y=0,Z=0)
    DriverItem=(Type=Bone,Name="thigh_l")
    FalloffRegionScaleFactors=(PositiveWidth=1,NegativeWidth=1.3,PositiveHeight=1,NegativeHeight=1)
    FalloffSize=0.5
    OptionalParentItem=(Type=Bone,Name="pelvis")
    RotationOffset=(X=120,Y=0,Z=-20)
    ExecutePin=@RigVMFunction_Sequence.B
   )
   > n2 = &RigUnit_SphericalPoseReader(
    ActiveRegionScaleFactors=(PositiveWidth=1,NegativeWidth=0,PositiveHeight=1,NegativeHeight=1)
    ActiveRegionSize=0.3
    Debug=(bDrawDebug=true,bDraw2D=false,bDrawLocalAxes=false,DebugScale=30,DebugSegments=30,DebugThickness=0.2)
    DriverAxis=(X=1,Y=0,Z=0)
    DriverItem=(Type=Bone,Name="thigh_r")
    FalloffRegionScaleFactors=(PositiveWidth=1,NegativeWidth=1.3,PositiveHeight=1,NegativeHeight=1)
    FalloffSize=0.5
    OptionalParentItem=(Type=Bone,Name="pelvis")
    RotationOffset=(X=-120,Y=180,Z=-20)
   )
   n3 = RigVMDispatchNode(
    Condition=/Script/RigVM.RigVMFunction_MathDoubleGreater(A=@n2.OutputParam, B=@n1.OutputParam) {is=struct RigVMUnitNode, template="Greater::Execute(in A,in B,out Result)"}.Result
    True=@n2.OutputParam
    False=@n1.OutputParam
   ) {template="DISPATCH_RigVMDispatch_If(in Condition,in True,in False,out Result)"}
   > &RigUnit_ModifyTransforms(
    ItemToModify=((Item=(Type=Bone,Name="Base_Cloth"),Transform=(Rotation=(X=0,Y=0.4226,Z=0,W=0.9063),Translation=(X=0,Y=0,Z=0),Scale3D=(X=1,Y=1,Z=1))))
    Weight=@n3.Result
   )
   > &RigUnit_ModifyTransforms(
    ItemToModify=((Item=(Type=Bone,Name="Base_front_belt"),Transform=(Rotation=(X=0,Y=0.1736,Z=0,W=0.9848),Translation=(X=-5,Y=0,Z=-5),Scale3D=(X=1,Y=1,Z=1))))
    Weight=@n3.Result
   )
   > &RigUnit_ModifyTransforms(
    ItemToModify=((Item=(Type=Bone,Name="Base_Cloth"),Transform=(Rotation=(X=-0.5,Y=0,Z=0,W=0.866),Translation=(X=0,Y=0,Z=0),Scale3D=(X=1,Y=1,Z=1))))
    Weight=@n1.OutputParam
   )
   > &RigUnit_ModifyTransforms(
    ItemToModify=((Item=(Type=Bone,Name="Base_Cloth"),Transform=(Rotation=(X=0.5,Y=0,Z=0,W=0.866),Translation=(X=0,Y=0,Z=0),Scale3D=(X=1,Y=1,Z=1))))
    Weight=@n2.OutputParam
   )
   > &RigUnit_ModifyTransforms(
    ItemToModify=((Item=(Type=Bone,Name="Base_left_belt"),Transform=(Rotation=(X=0,Y=0,Z=0,W=-1),Translation=(X=0,Y=0,Z=15),Scale3D=(X=1,Y=1,Z=1))))
    Weight=@n1.OutputParam
   )
   > &RigUnit_ModifyTransforms(
    ItemToModify=((Item=(Type=Bone,Name="Base_right_belt"),Transform=(Rotation=(X=0,Y=0,Z=0,W=-1),Translation=(X=-1,Y=0,Z=0),Scale3D=(X=1,Y=1,Z=1))))
    Weight=1
   )
   > &RigUnit_ModifyTransforms(
    ItemToModify=((Item=(Type=Bone,Name="Base_right_belt"),Transform=(Rotation=(X=0,Y=0,Z=0,W=-1),Translation=(X=0,Y=0,Z=-15),Scale3D=(X=1,Y=1,Z=1))))
    Weight=@n2.OutputParam
   )
   > &RigUnit_ParentConstraint(
    AdvancedSettings=(InterpolationType=Average,RotationOrderForFilter=XZY)
    Child=(Type=Bone,Name="Belt_metal_L")
    Parents="((Item=(Type=Bone,Name=\"pelvis\"),Weight=0.2649),(Item=(Type=Bone,Name=\"deform_groin_l\"),Weight=0.1529),(Item=(Type=Bone,Name=\"Base_front_belt\"),Weight=0.1315),(Item=(Type=Bone,Name=\"Base_left_belt\"),Weight=0.2867),(Item=(Type=Bone,Name=\"thigh_twist_01_l\"), ...(truncated)"
    Weight=1
   )
   > &RigUnit_ModifyTransforms(
    ItemToModify=((Item=(Type=Bone,Name="Belt_metal_L"),Transform=(Rotation=(X=-0.4226,Y=0,Z=0,W=0.9063),Translation=(X=0,Y=0,Z=0),Scale3D=(X=1,Y=1,Z=1))))
    Weight=@n1.OutputParam
   )
   > &RigUnit_ParentConstraint(
    AdvancedSettings=(InterpolationType=Average,RotationOrderForFilter=XZY)
    Child=(Type=Bone,Name="Belt_metal_R")
    Parents="((Item=(Type=Bone,Name=\"pelvis\"),Weight=0.2168),(Item=(Type=Bone,Name=\"deform_groin_r\"),Weight=0.1897),(Item=(Type=Bone,Name=\"Base_front_belt\"),Weight=0.1483),(Item=(Type=Bone,Name=\"Base_right_belt\"),Weight=0.2902),(Item=(Type=Bone,Name=\"thigh_twist_01_r\") ...(truncated)"
    Weight=1
   )
   > &RigUnit_ModifyTransforms(
    ItemToModify=((Item=(Type=Bone,Name="Belt_metal_R"),Transform=(Rotation=(X=0.4226,Y=0,Z=0,W=0.9063),Translation=(X=0,Y=0,Z=0),Scale3D=(X=1,Y=1,Z=1))))
    Weight=@n2.OutputParam
   )
  ]}
]
hierarchy [
 root {loc=<0 0 0>, rot=<0 0 0 1>}
 pelvis {parent=root, loc=<0 -0.5301 98.0571>, rot=<0.0127 0.707 -0.0127 -0.707>}
 spine_01 {parent=pelvis, loc=<6.1208 -0.9328 0>, rot=<0 0 0.034 -0.9994>}
 spine_02 {parent=spine_01, loc=<6.5997 -0.2095 0>, rot=<0 0 0.0238 -0.9997>}
 spine_03 {parent=spine_02, loc=<6.4914 -0.1237 0>, rot=<0 0 0.0149 -0.9999>}
 spine_04 {parent=spine_03, loc=<6.9598 0.0794 0>, rot=<0 0 -0.0132 -0.9999>}
 spine_05 {parent=spine_04, loc=<6.9987 0 0>, rot=<0 0 -0.0015 -1>}
 neck_01 {parent=spine_05, loc=<11.3138 0.1529 0>, rot=<0 0 -0.2003 -0.9797>}
 neck_02 {parent=neck_01, loc=<4.5627 0 0>, rot=<0 0 0.0504 -0.9987>}
 head {parent=neck_02, loc=<5.3914 0 0>, rot=<0 0 0.115 -0.9934>}
 Base_hair_l {parent=head, loc=<8.6946 -5.0438 -9.5225>, rot=<-0.0514 -0.8531 0.5184 0.0291>}
 Dyn_hair_l_01 {parent=Base_hair_l, loc=<0 0 0>, rot=<0 0 0 1>}
 Dyn_hair_l_02 {parent=Dyn_hair_l_01, loc=<12.5713 0 0>, rot=<0 -0.3364 0 -0.9417>}
 Dyn_hair_l_03 {parent=Dyn_hair_l_02, loc=<10.5442 0 0>, rot=<0.3418 0.1989 -0.0743 -0.9155>}
 Dyn_hair_l_04 {parent=Dyn_hair_l_03, loc=<8.5658 0 0>, rot=<0 0 0 1>}
 Base_hair_r {parent=head, loc=<8.6948 -5.0438 9.5225>, rot=<-0.0514 -0.8531 -0.5184 -0.0291>}
 Dyn_hair_r_01 {parent=Base_hair_r, loc=<0 0 0>, rot=<0 0 0 1>}
 Dyn_hair_r_02 {parent=Dyn_hair_r_01, loc=<12.5714 0 0>, rot=<0 0.3364 0 -0.9417>}
 Dyn_hair_r_03 {parent=Dyn_hair_r_02, loc=<10.5443 0 0>, rot=<-0.3418 -0.1989 -0.0743 -0.9155>}
 Dyn_hair_r_04 {parent=Dyn_hair_r_03, loc=<8.5662 0 0>, rot=<0 0 0 1>}
 Base_accessory_back {parent=head, loc=<5.3874 -11.3495 7.7497>, rot=<0.0041 -1 0 0>}
 Dyn_accessory_back_01 {parent=Base_accessory_back, loc=<0 0 0>, rot=<0 0 0 1>}
 Dyn_accessory_back_02 {parent=Dyn_accessory_back_01, loc=<4.8566 0 -0.0399>, rot=<0 0 0 1>}
 Dyn_accessory_back_03 {parent=Dyn_accessory_back_02, loc=<4.4248 0 0>, rot=<0 0 0 1>}
 Base_hair_back {parent=head, loc=<8.7197 -11.322 -0.3702>, rot=<0.0041 -1 0 0>}
 Dyn_hair_back_01 {parent=Base_hair_back, loc=<0 0 0>, rot=<0 0 0 1>}
 Dyn_hair_back_02 {parent=Dyn_hair_back_01, loc=<11.3016 0 -0.0399>, rot=<0 0 0 1>}
 Dyn_hair_back_03 {parent=Dyn_hair_back_02, loc=<10.241 0 0>, rot=<0 0 0 1>}
 FACIAL_C_FacialRoot {parent=head, loc=<2.1524 -0.746 -0.0002>, rot=<0.4879 -0.5118 -0.4879 -0.5118>}
 FACIAL_C_Forehead {parent=FACIAL_C_FacialRoot, loc=<-0.0005 -4.1498 8.6627>, rot=<-0.0819 0 -0.0001 -0.9966>}
 FACIAL_L_ForeheadIn {parent=FACIAL_C_FacialRoot, loc=<2.0936 -4.1748 8.4745>, rot=<-0.099 -0.0697 0.0014 -0.9926>}
 FACIAL_R_ForeheadIn {parent=FACIAL_C_FacialRoot, loc=<-2.0926 -4.1738 8.4702>, rot=<-0.099 0.0697 -0.0017 -0.9926>}
 FACIAL_L_ForeheadMid {parent=FACIAL_C_FacialRoot, loc=<4.02 -4.3824 7.7902>, rot=<-0.1087 -0.1747 -0.0275 -0.9782>}
 FACIAL_R_ForeheadMid {parent=FACIAL_C_FacialRoot, loc=<-4.0171 -4.3815 7.7862>, rot=<-0.1088 0.1747 0.0272 -0.9782>}
 FACIAL_L_ForeheadOut {parent=FACIAL_C_FacialRoot, loc=<5.9023 -4.12 6.6138>, rot=<-0.1581 -0.2524 -0.0891 -0.9505>}
 FACIAL_R_ForeheadOut {parent=FACIAL_C_FacialRoot, loc=<-5.9004 -4.1186 6.6117>, rot=<-0.1582 0.2523 0.0888 -0.9505>}
 FACIAL_L_EyesackUpper {parent=FACIAL_C_FacialRoot, loc=<4.1093 -3.4388 8.3504>, rot=<0.0105 -0.0817 -0.0679 -0.9943>}
 FACIAL_R_EyesackUpper {parent=FACIAL_C_FacialRoot, loc=<-4.1082 -3.4409 8.3478>, rot=<0.0105 0.0817 0.0676 -0.9943>}
 FACIAL_L_EyelidUpperFurrow {parent=FACIAL_C_FacialRoot, loc=<3.9205 -3.1252 8.25>, rot=<-0.068 -0.1476 -0.0855 -0.983>}
 FACIAL_R_EyelidUpperFurrow {parent=FACIAL_C_FacialRoot, loc=<-3.9156 -3.1264 8.2423>, rot=<-0.068 0.1476 0.0853 -0.983>}
 FACIAL_L_EyelidUpperA {parent=FACIAL_C_FacialRoot, loc=<3.7192 -1.3576 5.0637>, rot=<-0.1858 0 -0.0001 -0.9826>}
 FACIAL_R_EyelidUpperA {parent=FACIAL_C_FacialRoot, loc=<-3.713 -1.3626 5.0655>, rot=<-0.1858 0 -0.0001 -0.9826>}
 FACIAL_L_Eye {parent=FACIAL_C_FacialRoot, loc=<3.7011 -1.4052 4.996>, rot=<-0.0385 0 -0.0001 -0.9993>}
 FACIAL_R_Eye {parent=FACIAL_C_FacialRoot, loc=<-3.7022 -1.4045 4.9973>, rot=<-0.0386 0 -0.0001 -0.9993>}
 FACIAL_L_EyelidLowerA {parent=FACIAL_C_FacialRoot, loc=<3.7192 -1.3576 5.0637>, rot=<0.2468 0 -0.0001 -0.9691>}
 FACIAL_R_EyelidLowerA {parent=FACIAL_C_FacialRoot, loc=<-3.713 -1.3626 5.0655>, rot=<0.2468 0 -0.0001 -0.9691>}
 FACIAL_L_EyeCornerInner {parent=FACIAL_C_FacialRoot, loc=<1.7474 -0.9969 6.834>, rot=<-0.0385 0 -0.0001 -0.9993>}
 FACIAL_R_EyeCornerInner {parent=FACIAL_C_FacialRoot, loc=<-1.7381 -1.0066 6.8358>, rot=<-0.0386 0 -0.0001 -0.9993>}
 FACIAL_L_EyeCornerOuter {parent=FACIAL_C_FacialRoot, loc=<6.2709 -1.1845 6.396>, rot=<-0.0356 -0.3823 -0.0149 -0.9232>}
 FACIAL_R_EyeCornerOuter {parent=FACIAL_C_FacialRoot, loc=<-6.2705 -1.1689 6.41>, rot=<-0.0357 0.3823 0.0146 -0.9232>}
 FACIAL_L_EyesackLower {parent=FACIAL_C_FacialRoot, loc=<4.4169 0.7848 6.4404>, rot=<-0.0604 -0.1387 -0.0278 -0.9881>}
 FACIAL_R_EyesackLower {parent=FACIAL_C_FacialRoot, loc=<-4.4199 0.7743 6.4437>, rot=<-0.0604 0.1387 0.0275 -0.9881>}
 FACIAL_L_CheekInner {parent=FACIAL_C_FacialRoot, loc=<2.4484 1.3121 7.1808>, rot=<-0.1604 -0.0804 -0.2005 -0.9631>}
 FACIAL_R_CheekInner {parent=FACIAL_C_FacialRoot, loc=<-2.4484 1.307 7.1815>, rot=<-0.1605 0.0803 0.2002 -0.9632>}
 FACIAL_L_CheekOuter {parent=FACIAL_C_FacialRoot, loc=<5.1375 1.5001 5.3506>, rot=<0.0198 -0.3805 0.0856 -0.9206>}
 FACIAL_R_CheekOuter {parent=FACIAL_C_FacialRoot, loc=<-5.1377 1.4891 5.3494>, rot=<0.0197 0.3805 -0.0859 -0.9206>}
 FACIAL_L_NasolabialBulge {parent=FACIAL_C_FacialRoot, loc=<3.9746 4.7074 5.1902>, rot=<0.0223 -0.378 0.0091 -0.9255>}
 FACIAL_R_NasolabialBulge {parent=FACIAL_C_FacialRoot, loc=<-3.9776 4.7084 5.193>, rot=<0.0222 0.378 -0.0094 -0.9255>}
 FACIAL_L_NasolabialFurrow {parent=FACIAL_C_FacialRoot, loc=<3.3721 3.5996 7.8737>, rot=<-0.0688 -0.1117 -0.3683 -0.9204>}
 FACIAL_R_NasolabialFurrow {parent=FACIAL_C_FacialRoot, loc=<-3.3749 3.5959 7.87>, rot=<-0.0688 0.1117 0.3681 -0.9205>}
 FACIAL_L_CheekLower {parent=FACIAL_C_FacialRoot, loc=<4.4861 4.584 4.5117>, rot=<-0.0001 -0.4288 0.0411 -0.9025>}
 FACIAL_R_CheekLower {parent=FACIAL_C_FacialRoot, loc=<-4.4889 4.5857 4.5159>, rot=<-0.0003 0.4288 -0.0414 -0.9025>}
 FACIAL_C_Nose {parent=FACIAL_C_FacialRoot, loc=<-0.0001 0.3921 8.8586>, rot=<-0.1685 0 -0.0001 -0.9857>}
 FACIAL_C_NoseLower {parent=FACIAL_C_Nose, loc=<0.0007 3.4681 -0.5275>, rot=<0.1305 0 0 -0.9914>}
 FACIAL_L_Nostril {parent=FACIAL_C_Nose, loc=<0.9447 2.4153 -0.3097>, rot=<0.0173 0.1305 0.0023 -0.9913>}
 FACIAL_R_Nostril {parent=FACIAL_C_Nose, loc=<-0.9343 2.407 -0.3159>, rot=<0.0173 -0.1305 -0.0023 -0.9913>}
 FACIAL_C_NoseTip {parent=FACIAL_C_Nose, loc=<-0.0003 2.4331 0.6282>, rot=<0.0872 0 0 -0.9962>}
 FACIAL_C_MouthUpper {parent=FACIAL_C_FacialRoot, loc=<-0.0002 5.0356 6.4073>, rot=<-0.0385 0 -0.0001 -0.9993>}
 FACIAL_C_LipUpper {parent=FACIAL_C_MouthUpper, loc=<0.0023 0.0468 3.1772>, rot=<-0.0872 0 0 -0.9962>}
 FACIAL_L_LipUpper {parent=FACIAL_C_MouthUpper, loc=<1.371 0.1015 2.7537>, rot=<-0.0778 -0.1601 -0.0379 -0.9833>}
 FACIAL_R_LipUpper {parent=FACIAL_C_MouthUpper, loc=<-1.3698 0.1039 2.752>, rot=<-0.0778 0.1601 0.0379 -0.9833>}
 FACIAL_L_LipUpperOuter {parent=FACIAL_C_MouthUpper, loc=<2.3356 0.2986 1.8725>, rot=<-0.0984 -0.2293 -0.0983 -0.9634>}
 FACIAL_R_LipUpperOuter {parent=FACIAL_C_MouthUpper, loc=<-2.3316 0.2989 1.8713>, rot=<-0.0984 0.2293 0.0983 -0.9634>}
 FACIAL_L_LipCorner {parent=FACIAL_C_MouthUpper, loc=<2.8141 0.7358 0.839>, rot=<0 -0.2078 0 -0.9782>}
 FACIAL_R_LipCorner {parent=FACIAL_C_MouthUpper, loc=<-2.8174 0.7362 0.8287>, rot=<0 0.2078 0 -0.9782>}
 FACIAL_L_Masseter {parent=FACIAL_C_FacialRoot, loc=<7.7865 1.8349 2.8486>, rot=<0.0311 0.1125 0.0121 -0.9931>}
 FACIAL_R_Masseter {parent=FACIAL_C_FacialRoot, loc=<-7.7847 1.8317 2.8532>, rot=<0.0311 -0.1125 -0.0124 -0.9931>}
 FACIAL_C_UnderChin {parent=FACIAL_C_FacialRoot, loc=<-0.0003 8.8775 2.5588>, rot=<0.4661 0.0001 -0.0001 -0.8847>}
 FACIAL_L_UnderChin {parent=FACIAL_C_FacialRoot, loc=<3.0607 7.4128 1.7997>, rot=<0.3699 -0.3577 0.0685 -0.8547>}
 FACIAL_R_UnderChin {parent=FACIAL_C_FacialRoot, loc=<-3.0612 7.4064 1.8153>, rot=<0.3698 0.3578 -0.0687 -0.8547>}
 FACIAL_C_TeethUpper {parent=FACIAL_C_FacialRoot, loc=<0.0026 5.8217 8.2451>, rot=<-0.0385 0 -0.0001 -0.9993>}
 FACIAL_C_LowerLipRotation {parent=FACIAL_C_FacialRoot, loc=<-0.0021 2.4168 1.1556>, rot=<0.3056 0.0001 -0.0001 -0.9522>}
 FACIAL_C_MouthLower {parent=FACIAL_C_LowerLipRotation, loc=<0.0027 -0.9264 5.7949>, rot=<-0.342 0 0 -0.9397>}
 FACIAL_C_LipLower {parent=FACIAL_C_MouthLower, loc=<0.0037 2.0118 2.2769>, rot=<0.0872 0 0 -0.9962>}
 FACIAL_L_LipLower {parent=FACIAL_C_MouthLower, loc=<1.1445 1.8966 2.0057>, rot=<0.0927 -0.1294 0.0272 -0.9869>}
 FACIAL_R_LipLower {parent=FACIAL_C_MouthLower, loc=<-1.1369 1.8966 2.0071>, rot=<0.0927 0.1294 -0.0272 -0.9869>}
 FACIAL_L_LipLowerOuter {parent=FACIAL_C_MouthLower, loc=<2.1477 1.3988 1.3396>, rot=<0.0913 -0.2321 0.0318 -0.9679>}
 FACIAL_R_LipLowerOuter {parent=FACIAL_C_MouthLower, loc=<-2.1449 1.4014 1.3418>, rot=<0.0913 0.2321 -0.0318 -0.9679>}
 FACIAL_C_TeethLower {parent=FACIAL_C_LowerLipRotation, loc=<0.0057 -1.3288 7.6614>, rot=<-0.342 0 0 -0.9397>}
 FACIAL_C_Tongue1 {parent=FACIAL_C_TeethLower, loc=<-0.0047 -0.6288 -4.985>, rot=<-0.2588 0 0 -0.9659>}
 FACIAL_C_Tongue2 {parent=FACIAL_C_Tongue1, loc=<0.001 0.5882 1.3763>, rot=<0.2588 0 0 -0.9659>}
 FACIAL_C_Tongue3 {parent=FACIAL_C_Tongue2, loc=<0.0007 0.3768 1.2907>, rot=<0.1305 0 0 -0.9914>}
 FACIAL_C_Tongue4 {parent=FACIAL_C_Tongue3, loc=<0.001 -0.0164 1.1675>, rot=<0.0436 0 0 -0.999>}
 FACIAL_C_TongueUpper3 {parent=FACIAL_C_Tongue3, loc=<-0.0001 -0.3887 -0.1347>, rot=<0 0 0 1>}
 FACIAL_C_TongueLower3 {parent=FACIAL_C_Tongue3, loc=<0.0004 0.869 -0.0176>, rot=<0 0 0 1>}
 FACIAL_L_TongueSide3 {parent=FACIAL_C_Tongue3, loc=<1.6359 0.0765 0.1593>, rot=<0 0 0 1>}
 FACIAL_R_TongueSide3 {parent=FACIAL_C_Tongue3, loc=<-1.6371 0.0759 0.1585>, rot=<0 0 0 1>}
 FACIAL_C_TongueUpper2 {parent=FACIAL_C_Tongue2, loc=<0 -0.6082 -0.0416>, rot=<0 0 0 1>}
 FACIAL_L_TongueSide2 {parent=FACIAL_C_Tongue2, loc=<1.9494 0.1792 0.2837>, rot=<0 0 0 1>}
 FACIAL_R_TongueSide2 {parent=FACIAL_C_Tongue2, loc=<-1.9508 0.1785 0.2834>, rot=<0 0 0 1>}
 FACIAL_C_Jaw {parent=FACIAL_C_FacialRoot, loc=<-0.0021 2.4167 1.1556>, rot=<0.3056 0.0001 -0.0001 -0.9522>}
 FACIAL_C_Jawline {parent=FACIAL_C_Jaw, loc=<0.004 3.0085 7.4028>, rot=<0.2586 0 0 -0.966>}
 FACIAL_L_ChinSide {parent=FACIAL_C_Jaw, loc=<3.3099 1.0775 7.2629>, rot=<-0.2208 -0.3738 0.0499 -0.8995>}
 FACIAL_R_ChinSide {parent=FACIAL_C_Jaw, loc=<-3.307 1.0797 7.2635>, rot=<-0.2208 0.3738 -0.0499 -0.8995>}
 FACIAL_C_Chin {parent=FACIAL_C_Jaw, loc=<0.0008 1.3077 7.7629>, rot=<-0.2591 0 0 -0.9659>}
 FACIAL_L_Ear {parent=FACIAL_C_FacialRoot, loc=<7.2001 -0.5498 -0.7673>, rot=<-0.0648 0.0003 -0.0263 -0.9976>}
 FACIAL_R_Ear {parent=FACIAL_C_FacialRoot, loc=<-7.2087 -0.5459 -0.7457>, rot=<-0.0648 -0.0003 0.026 -0.9976>}
 FACIAL_L_Jawline {parent=FACIAL_C_FacialRoot, loc=<5.4822 5.8576 3.8748>, rot=<0.172 0.1598 -0.0487 -0.9708>}
 FACIAL_R_Jawline {parent=FACIAL_C_FacialRoot, loc=<-5.4906 5.8624 3.8868>, rot=<0.1721 -0.1597 0.0484 -0.9709>}
 FACIAL_C_NoseBridge {parent=FACIAL_C_FacialRoot, loc=<-0.0005 -2.0421 9.5453>, rot=<0.0053 0 -0.0001 -1>}
 FACIAL_L_EyelidUpperB {parent=FACIAL_C_FacialRoot, loc=<3.7192 -1.3576 5.0637>, rot=<-0.2623 0 -0.0001 -0.965>}
 FACIAL_L_EyelidLowerB {parent=FACIAL_C_FacialRoot, loc=<3.7192 -1.3576 5.0637>, rot=<0.2805 0.0001 -0.0001 -0.9599>}
 FACIAL_R_EyelidLowerB {parent=FACIAL_C_FacialRoot, loc=<-3.713 -1.3626 5.0655>, rot=<0.2889 0.0001 -0.0001 -0.9574>}
 FACIAL_R_EyelidUpperB {parent=FACIAL_C_FacialRoot, loc=<-3.713 -1.3626 5.0655>, rot=<-0.2624 0 -0.0001 -0.965>}
 clavicle_l {parent=spine_05, loc=<5.4831 -0.1053 -3.2946>, rot=<0.0465 -0.7289 -0.009 -0.6829>}
 upperarm_l {parent=clavicle_l, loc=<9.8658 -1.2136 -0.4105>, rot=<0.0164 -0.3245 -0.0269 -0.9454>}
 lowerarm_l {parent=upperarm_l, loc=<26.465 -0.8345 0.259>, rot=<0.0036 0.0161 -0.2298 -0.9731>}
 hand_l {parent=lowerarm_l, loc=<22.7355 -0.3203 -0.6105>, rot=<-0.6784 -0.0223 0.0735 -0.7307>}
 thumb_01_l {parent=hand_l, loc=<1.3118 -0.3191 -2.459>, rot=<0.6546 -0.4042 0.0664 -0.6354>}
 thumb_02_l {parent=thumb_01_l, loc=<5.9336 -0.0448 0.0401>, rot=<-0.0037 0.0213 0.298 -0.9543>}
 thumb_03_l {parent=thumb_02_l, loc=<2.2223 0.3682 -0.0907>, rot=<0.0118 -0.013 0.0066 -0.9998>}
 index_metacarpal_l {parent=hand_l, loc=<1.6535 0.0183 -2.96>, rot=<0.0158 -0.0124 -0.0173 -0.9996>}
 index_01_l {parent=index_metacarpal_l, loc=<8.3008 0.0357 -0.0524>, rot=<-0.0044 0.0485 0.0924 -0.9945>}
 index_02_l {parent=index_01_l, loc=<3.3587 -0.3004 0.0452>, rot=<-0.002 0.0027 0.3225 -0.9466>}
 index_03_l {parent=index_02_l, loc=<2.5578 -0.123 -0.0252>, rot=<-0.002 0.0067 0.2558 -0.9667>}
 middle_metacarpal_l {parent=hand_l, loc=<2.1856 -0.1725 -0.5984>, rot=<0.006 0.0168 -0.0329 -0.9993>}
 middle_01_l {parent=middle_metacarpal_l, loc=<7.3275 -0.2447 0.0129>, rot=<-0.0208 0.0042 0.0858 -0.9961>}
 middle_02_l {parent=middle_01_l, loc=<4.0123 0.1401 0.1175>, rot=<-0.0218 0.0258 0.3512 -0.9357>}
 middle_03_l {parent=middle_02_l, loc=<2.6077 0.0534 0.1373>, rot=<0.0082 0.0188 0.1708 -0.9851>}
 ring_metacarpal_l {parent=hand_l, loc=<2.2074 -0.3625 1.3638>, rot=<-0.0284 0.0722 -0.0458 -0.9959>}
 ring_01_l {parent=ring_metacarpal_l, loc=<7.0295 -0.6197 -0.1446>, rot=<-0.0489 -0.0252 0.1892 -0.9804>}
 ring_02_l {parent=ring_01_l, loc=<3.6174 0.4647 -0.1027>, rot=<0 0 0.2806 -0.9598>}
 ring_03_l {parent=ring_02_l, loc=<2.4152 0.114 0.0199>, rot=<0 0 0.1189 -0.9929>}
 pinky_metacarpal_l {parent=hand_l, loc=<1.7616 -0.9753 3.3814>, rot=<-0.0397 0.0814 -0.069 -0.9935>}
 pinky_01_l {parent=pinky_metacarpal_l, loc=<7.1709 -0.4831 -0.1094>, rot=<-0.0654 -0.0035 0.2242 -0.9723>}
 pinky_02_l {parent=pinky_01_l, loc=<2.8274 0.0893 -0.1043>, rot=<0.0037 0.0011 0.2854 -0.9584>}
 pinky_03_l {parent=pinky_02_l, loc=<2.2046 0.0307 -0.0078>, rot=<-0.0002 0.0098 0.1445 -0.9895>}
 deform_wrist_dn_base_l {parent=hand_l, loc=<0 -1.5326 0>, rot=<0 0 0 -1>}
 deform_wrist_dn_l {parent=deform_wrist_dn_base_l, loc=<0 0 0>, rot=<0 0 0 1>}
 deform_wrist_up_base_l {parent=hand_l, loc=<0 2.3009 0>, rot=<0 0 0 -1>}
 deform_wrist_up_l {parent=deform_wrist_up_base_l, loc=<0 0 0>, rot=<0 0 0 1>}
 hand_attach_l {parent=hand_l, loc=<8.5 -3 0>, rot=<-0.7071 -0.7071 0 0.0002>}
 weapon_l {parent=hand_l, loc=<-0.5517 -2.3491 5.2893>, rot=<0.0002 0 0 -1>}
 lowerarm_twist_02_l {parent=lowerarm_l, loc=<10.6707 0 0>, rot=<0 0 0 1>}
 lowerarm_twist_01_l {parent=lowerarm_l, loc=<18.316 0 0>, rot=<0 0 0 1>}
 deform_elbow_in_base_l {parent=lowerarm_l, loc=<0.1834 3.6062 -0.2572>, rot=<0 0 0 -1>}
 deform_elbow_in_l {parent=deform_elbow_in_base_l, loc=<0 0 0>, rot=<0 0 0 1>}
 deform_elbow_out_base_l {parent=lowerarm_l, loc=<0 -4 0>, rot=<0 0 0 1>}
 deform_elbow_out_l {parent=deform_elbow_out_base_l, loc=<0 0 0>, rot=<0 0 0 1>}
 upperarm_twist_01_l {parent=upperarm_l, loc=<7.2135 -0.1134 0>, rot=<0 0 0 1>}
 deform_upperarm_bk_base_l {parent=upperarm_twist_01_l, loc=<-6.4053 -4.3746 0.314>, rot=<0 0 0 -1>}
 deform_upperarm_bk_l {parent=deform_upperarm_bk_base_l, loc=<0 0 0>, rot=<0 0 0 1>}
 deform_upperarm_fr_base_l {parent=upperarm_twist_01_l, loc=<-5.6312 5.092 0.3354>, rot=<0 0 0 -1>}
 deform_upperarm_fr_l {parent=deform_upperarm_fr_base_l, loc=<0 0 0>, rot=<0 0 0 1>}
 deform_upperarm_up_base_l {parent=upperarm_twist_01_l, loc=<-8.891 0.4825 3.9517>, rot=<0 0 0 -1>}
 deform_upperarm_up_l {parent=deform_upperarm_up_base_l, loc=<0 0 0>, rot=<0 0 0 1>}
 upperarm_twist_02_l {parent=upperarm_l, loc=<19.7998 -0.1134 0>, rot=<0 0 0 1>}
 clavicle_r {parent=spine_05, loc=<5.4831 -0.1053 3.2946>, rot=<0.7289 0.0465 -0.6829 0.009>}
 upperarm_r {parent=clavicle_r, loc=<-9.8658 1.2136 0.4105>, rot=<0.0164 -0.3245 -0.0269 -0.9454>}
 lowerarm_r {parent=upperarm_r, loc=<-26.465 0.8345 -0.259>, rot=<0.0036 0.0161 -0.2298 -0.9731>}
 hand_r {parent=lowerarm_r, loc=<-22.7355 0.3203 0.6106>, rot=<-0.6784 -0.0223 0.0735 -0.7307>}
 thumb_01_r {parent=hand_r, loc=<-1.3118 0.3191 2.459>, rot=<0.6546 -0.4042 0.0664 -0.6354>}
 thumb_02_r {parent=thumb_01_r, loc=<-5.9336 0.0448 -0.0401>, rot=<-0.0037 0.0213 0.298 -0.9543>}
 thumb_03_r {parent=thumb_02_r, loc=<-2.2223 -0.3682 0.0907>, rot=<0.0118 -0.013 0.0066 -0.9998>}
 index_metacarpal_r {parent=hand_r, loc=<-1.6535 -0.0183 2.96>, rot=<0.0158 -0.0124 -0.0173 -0.9996>}
 index_01_r {parent=index_metacarpal_r, loc=<-8.3008 -0.0357 0.0524>, rot=<-0.0044 0.0485 0.0924 -0.9945>}
 index_02_r {parent=index_01_r, loc=<-3.3587 0.3004 -0.0452>, rot=<-0.002 0.0027 0.3225 -0.9466>}
 index_03_r {parent=index_02_r, loc=<-2.5578 0.123 0.0252>, rot=<-0.002 0.0067 0.2558 -0.9667>}
 middle_metacarpal_r {parent=hand_r, loc=<-2.1856 0.1725 0.5984>, rot=<0.006 0.0168 -0.0329 -0.9993>}
 middle_01_r {parent=middle_metacarpal_r, loc=<-7.3275 0.2447 -0.0129>, rot=<-0.0208 0.0042 0.0858 -0.9961>}
 middle_02_r {parent=middle_01_r, loc=<-4.0123 -0.1401 -0.1175>, rot=<-0.0218 0.0258 0.3512 -0.9357>}
 middle_03_r {parent=middle_02_r, loc=<-2.6077 -0.0534 -0.1373>, rot=<0.0082 0.0188 0.1708 -0.9851>}
 ring_metacarpal_r {parent=hand_r, loc=<-2.2074 0.3625 -1.3638>, rot=<-0.0284 0.0722 -0.0458 -0.9959>}
 ring_01_r {parent=ring_metacarpal_r, loc=<-7.0295 0.6197 0.1446>, rot=<-0.0489 -0.0252 0.1892 -0.9804>}
 ring_02_r {parent=ring_01_r, loc=<-3.6174 -0.4647 0.1027>, rot=<0 0 0.2806 -0.9598>}
 ring_03_r {parent=ring_02_r, loc=<-2.4152 -0.114 -0.0199>, rot=<0 0 0.1189 -0.9929>}
 pinky_metacarpal_r {parent=hand_r, loc=<-1.7616 0.9753 -3.3814>, rot=<-0.0397 0.0814 -0.069 -0.9935>}
 pinky_01_r {parent=pinky_metacarpal_r, loc=<-7.1709 0.4831 0.1094>, rot=<-0.0654 -0.0035 0.2242 -0.9723>}
 pinky_02_r {parent=pinky_01_r, loc=<-2.8274 -0.0893 0.1043>, rot=<0.0037 0.0011 0.2854 -0.9584>}
 pinky_03_r {parent=pinky_02_r, loc=<-2.2046 -0.0307 0.0078>, rot=<-0.0002 0.0098 0.1445 -0.9895>}
 deform_wrist_dn_base_r {parent=hand_r, loc=<-0.0001 1.5328 0>, rot=<0 0 0 -1>}
 deform_wrist_dn_r {parent=deform_wrist_dn_base_r, loc=<0 0 0>, rot=<0 0 0 1>}
 deform_wrist_up_base_r {parent=hand_r, loc=<-0.0002 -2.3007 -0.0001>, rot=<0 0 0 -1>}
 deform_wrist_up_r {parent=deform_wrist_up_base_r, loc=<0 0 0>, rot=<0 0 0 1>}
 weapon_r {parent=hand_r, loc=<1.8709 4.2805 -4.9512>, rot=<-0.1449 -0.0495 0.9873 -0.0433>}
 hand_attach_r {parent=hand_r, loc=<-8.5001 2.9999 0>, rot=<0 0 -0.7071 -0.7071>}
 lowerarm_twist_02_r {parent=lowerarm_r, loc=<-10.6707 0 0>, rot=<0 0 0 -1>}
 lowerarm_twist_01_r {parent=lowerarm_r, loc=<-18.316 0 0>, rot=<0 0 0 -1>}
 deform_elbow_in_base_r {parent=lowerarm_r, loc=<-0.1834 -3.6062 0.2572>, rot=<0 0 0 -1>}
 deform_elbow_in_r {parent=deform_elbow_in_base_r, loc=<0 0 0>, rot=<0 0 0 1>}
 deform_elbow_out_base_r {parent=lowerarm_r, loc=<0 4 -0.0001>, rot=<0 0 0 -1>}
 deform_elbow_out_r {parent=deform_elbow_out_base_r, loc=<0 0 0>, rot=<0 0 0 1>}
 upperarm_twist_01_r {parent=upperarm_r, loc=<-7.2135 0.1134 0>, rot=<0 0 0 -1>}
 deform_upperarm_bk_base_r {parent=upperarm_twist_01_r, loc=<6.4053 4.3746 -0.314>, rot=<0 0 0 -1>}
 deform_upperarm_bk_r {parent=deform_upperarm_bk_base_r, loc=<0 0 0>, rot=<0 0 0 1>}
 deform_upperarm_fr_base_r {parent=upperarm_twist_01_r, loc=<5.6312 -5.092 -0.3354>, rot=<0 0 0 -1>}
 deform_upperarm_fr_r {parent=deform_upperarm_fr_base_r, loc=<0 0 0>, rot=<0 0 0 1>}
 deform_upperarm_up_base_r {parent=upperarm_twist_01_r, loc=<8.891 -0.4825 -3.9517>, rot=<0 0 0 -1>}
 deform_upperarm_up_r {parent=deform_upperarm_up_base_r, loc=<0 0 0>, rot=<0 0 0 1>}
 upperarm_twist_02_r {parent=upperarm_r, loc=<-19.7998 0.1134 0>, rot=<0 0 0 -1>}
 attach_backpack {parent=spine_05, loc=<-0.1229 -13.473 0>, rot=<-0.0093 -0.707 0.0093 -0.707>}
 Backpack_BR {parent=attach_backpack, type=Null, loc=<0 0 0>, rot=<0 0 0 1>}
 attach_cape {parent=spine_05, loc=<9.017 -11.714 0>, rot=<0.0616 -0.7044 -0.0616 -0.7044>}
 cape {parent=attach_cape, type=Null, loc=<0 0 0>, rot=<0 0 0 1>}
 deform_pec_aim_base_l {parent=spine_04, loc=<3.2456 9.3309 -1.2237>, rot=<0.0603 -0.4033 0.3318 -0.8507>}
 deform_pec_aim_l {parent=deform_pec_aim_base_l, loc=<0 0 0>, rot=<0 0 0 1>}
 deform_pec_base_l {parent=deform_pec_aim_l, loc=<9.3665 0 0>, rot=<0 0 0 1>}
 deform_pec_l {parent=deform_pec_base_l, loc=<0 0 0>, rot=<0 0 0 1>}
 deform_pec_aim_base_r {parent=spine_04, loc=<3.2452 9.3309 1.2237>, rot=<-0.4033 -0.0603 0.8507 0.3318>}
 deform_pec_aim_r {parent=deform_pec_aim_base_r, loc=<0 0 0>, rot=<0 0 0 1>}
 deform_pec_base_r {parent=deform_pec_aim_r, loc=<-9.3667 -0.0003 -0.0004>, rot=<0 0 0 1>}
 deform_pec_r {parent=deform_pec_base_r, loc=<0 0 0>, rot=<0 0 0 1>}
 thigh_l {parent=pelvis, loc=<-1.0855 -0.2109 -8.7053>, rot=<0.0009 0.0355 0.0155 -0.9993>}
 calf_l {parent=thigh_l, loc=<-42.5284 -0.0019 -0.7297>, rot=<0.0006 -0.0161 -0.0391 -0.9991>}
 foot_l {parent=calf_l, loc=<-46.7401 -2.5014 -0.592>, rot=<0.0427 -0.0238 0.0353 -0.9982>}
 ball_l {parent=foot_l, loc=<-3.5658 10.8066 -0.1544>, rot=<0.0001 -0.0001 -0.7063 -0.7079>}
 calf_twist_02_l {parent=calf_l, loc=<-19.1546 -1.7943 -0.2638>, rot=<0.0237 0.0095 -0.0035 -0.9997>}
 calf_twist_01_l {parent=calf_l, loc=<-35.8483 -1.9197 -0.5794>, rot=<0.0237 0.0095 -0.0035 -0.9997>}
 deform_knee_in_base_l {parent=calf_l, loc=<0.8031 -5.0231 0.2456>, rot=<0 0 0 -1>}
 deform_knee_in_l {parent=deform_knee_in_base_l, loc=<0 0 0>, rot=<0 0 0 1>}
 deform_knee_out_base_l {parent=calf_l, loc=<0 4 0>, rot=<0 0 0 -1>}
 deform_knee_out_l {parent=deform_knee_out_base_l, loc=<0 0 0>, rot=<0 0 0 1>}
 thigh_twist_01_l {parent=thigh_l, loc=<-27.952 0 0>, rot=<0.0013 0 0 -1>}
 deform_groin_base_l {parent=thigh_l, loc=<0.4801 5.6776 -0.2486>, rot=<0 0 0 -1>}
 deform_groin_l {parent=deform_groin_base_l, loc=<0 0 0>, rot=<0 0 0 1>}
 thigh_r {parent=pelvis, loc=<-1.0855 -0.2109 8.7053>, rot=<0.0355 -0.0009 0.9993 0.0155>}
 calf_r {parent=thigh_r, loc=<42.5284 0.0019 0.7297>, rot=<0.0006 -0.0161 -0.0391 -0.9991>}
 foot_r {parent=calf_r, loc=<46.7401 2.5014 0.592>, rot=<0.0427 -0.0238 0.0353 -0.9982>}
 ball_r {parent=foot_r, loc=<3.5658 -10.8066 0.1544>, rot=<0.0001 -0.0001 -0.7063 -0.7079>}
 calf_twist_02_r {parent=calf_r, loc=<19.1546 1.7943 0.2638>, rot=<0.0237 0.0095 -0.0035 -0.9997>}
 calf_twist_01_r {parent=calf_r, loc=<35.8483 1.9197 0.5794>, rot=<0.0237 0.0095 -0.0035 -0.9997>}
 deform_knee_in_base_r {parent=calf_r, loc=<-0.8031 5.0231 -0.2457>, rot=<0 0 0 -1>}
 deform_knee_in_r {parent=deform_knee_in_base_r, loc=<0 0 0>, rot=<0 0 0 1>}
 deform_knee_out_base_r {parent=calf_r, loc=<0 -4 0>, rot=<0 0 0 -1>}
 deform_knee_out_r {parent=deform_knee_out_base_r, loc=<0 0 0>, rot=<0 0 0 1>}
 thigh_twist_01_r {parent=thigh_r, loc=<27.952 0 0>, rot=<0.0013 0 0 -1>}
 deform_groin_base_r {parent=thigh_r, loc=<-0.4801 -5.6776 0.2486>, rot=<0 0 0 -1>}
 deform_groin_r {parent=deform_groin_base_r, loc=<0 0 0>, rot=<0 0 0 1>}
 deform_glute_base_l {parent=pelvis, loc=<-7.3466 -9.0213 -9.1245>, rot=<0.0172 0.031 -0.4477 -0.8935>}
 deform_glute_l {parent=deform_glute_base_l, loc=<0 0 0>, rot=<0 0 0 1>}
 deform_glute_base_r {parent=pelvis, loc=<-7.3464 -9.0213 9.1245>, rot=<0.031 -0.0172 0.8935 -0.4477>}
 deform_glute_r {parent=deform_glute_base_r, loc=<0 0 0>, rot=<0 0 0 1>}
 Base_front_belt {parent=pelvis, loc=<-9.5323 9.5874 0.025>, rot=<-0.0193 -0.7184 0.6952 0.015>}
 Base_Cloth {parent=Base_front_belt, loc=<0 0 0>, rot=<0 0 0 1>}
 Dyn_cloth_01 {parent=Base_Cloth, loc=<0 0 0>, rot=<0 0 0 1>}
 Dyn_cloth_02 {parent=Dyn_cloth_01, loc=<5.8615 0.0309 -0.0692>, rot=<0 0.0089 0 -1>}
 Dyn_cloth_03 {parent=Dyn_cloth_02, loc=<9.0501 0.0138 0.0129>, rot=<0 0 0 1>}
 Base_left_belt {parent=pelvis, loc=<-3.246 3.6942 -14.8586>, rot=<-0.156 -0.6897 -0.1312 -0.6948>}
 Base_right_belt {parent=pelvis, loc=<-3.246 3.6942 14.8586>, rot=<-0.6897 0.156 0.6948 -0.1312>}
 Belt_metal_L {parent=pelvis, loc=<-1.431 7.721 -10.124>, rot=<-0.4758 -0.5817 -0.304 -0.5855>}
 Belt_metal_R {parent=pelvis, loc=<-1.431 7.721 10.124>, rot=<0.5389 0.3369 0.608 0.4759>}
 ik_hand_root {parent=root, loc=<0 0 -2.1991>, rot=<0 0 -0.3848 -0.923>}
 ik_hand_gun {parent=ik_hand_root, loc=<-48.1743 3.3545 104.4381>, rot=<0.6276 0.1309 0.4183 -0.6434>}
 ik_hand_r {parent=ik_hand_gun, loc=<0 0 0>, rot=<0 0 0 1>}
 ik_hand_l {parent=ik_hand_gun, loc=<59.3271 67.6997 34.3524>, rot=<-0.6158 -0.7027 -0.3565 0>}
 attach {parent=root, loc=<0 0 -2.1991>, rot=<0 0 -0.3848 -0.923>}
 ik_foot_root {parent=root, loc=<0 0 -2.1991>, rot=<0 0 -0.3848 -0.923>}
 ik_foot_l {parent=ik_foot_root, loc=<14.6934 8.7319 9.3915>, rot=<0.0728 0.7019 0.067 -0.7054>}
 ik_foot_r {parent=ik_foot_root, loc=<-10.5192 -25.4024 15.5724>, rot=<-0.6318 -0.0126 0.6518 -0.4193>}
 root_fp {parent=root, loc=<0 0 -2.1991>, rot=<0 0 -0.3848 -0.923>}
 attach_fp {parent=root_fp, loc=<0 0 0>, rot=<0 0 0 1>}
 camera_root {parent=root_fp, loc=<0 0 0>, rot=<0 0 0 1>}
 camera {parent=camera_root, loc=<0 0 0>, rot=<0 0 0 1>}
 attach_emote_01 {parent=root, loc=<0 0 -2.1991>, rot=<0 0 -0.3848 -0.923>}
 attach_emote_02 {parent=root, loc=<0 0 -2.1991>, rot=<0 0 -0.3848 -0.923>}
 attach_emote_03 {parent=root, loc=<0 0 -2.1991>, rot=<0 0 -0.3848 -0.923>}
 warp_point_01 {parent=root, loc=<0 0 -2.1991>, rot=<0 0 -0.3848 -0.923>}
 warp_point_02 {parent=root, loc=<0 0 -2.1991>, rot=<0 0 -0.3848 -0.923>}
 warp_point_03 {parent=root, loc=<0 0 -2.1991>, rot=<0 0 -0.3848 -0.923>}
]
