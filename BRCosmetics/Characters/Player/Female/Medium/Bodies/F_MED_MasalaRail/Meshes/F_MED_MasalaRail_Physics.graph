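# Physics asset for the F_MED_MasalaRail female player body cosmetic, setting up a dangling front-belt cloth piece as a chain of rigid bodies.
# Small capsules on Base_front_belt and Base_Cloth share a free-rotating joint, followed by three larger Dyn_cloth capsules linked in series.
# The cloth joints have tight twist and moderate, asymmetric swing limits.
# Capsules on both thighs have no constraints, apparently serving as leg colliders for the cloth.
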PhysicsAsset /BRCosmetics/Characters/Player/Female/Medium/Bodies/F_MED_MasalaRail/Meshes/F_MED_MasalaRail_Physics
bodies {
 Base_Cloth [{shapeType=Capsule, length=0.505, radius=0.505, shapeName=Base_Cloth_capsule, rotation=<rot 0 0 90>}]
 Base_front_belt [{shapeType=Capsule, length=4.4685, radius=1.2278, shapeName=Base_front_belt_capsule, rotation=<rot 0 0 90>}]
 Dyn_cloth_01 [{shapeType=Capsule, length=10.0962, radius=3, shapeName=Dyn_cloth_01_capsule, center=<3.6245 0 0>, rotation=<rot 0 0 90>}]
 Dyn_cloth_02 [{shapeType=Capsule, length=8.7167, radius=3.5, shapeName=Dyn_cloth_02_capsule, center=<3.6245 0 0>, rotation=<rot 0 0 90>}]
 Dyn_cloth_03 [{shapeType=Capsule, length=5.005, radius=4.005, shapeName=Dyn_cloth_03_capsule, center=<3 0 0>, rotation=<rot 0 0 90>}]
 thigh_l [{shapeType=Capsule, length=23.5164, radius=7.4036, center=<-20.8471 -1.1456 1.5729>, rotation=<rot 0 -84.3681 -88.1363>}]
 thigh_r [{shapeType=Capsule, length=23.4892, radius=7.4322, center=<20.7579 1.178 -1.7075>, rotation=<rot 0 -84.0082 -87.8948>}]
}
constraints [
 {
  bone1Name=Base_Cloth
  bone2Name=Base_front_belt
  swing1LimitDegrees=45
  swing1Motion=Free
  swing2LimitDegrees=45
  swing2Motion=Free
  twistLimitDegrees=45
  twistMotion=Free
 }
 {bone1Name=Dyn_cloth_01, bone2Name=Base_Cloth, swing1LimitDegrees=20, swing2LimitDegrees=40, twistLimitDegrees=1}
 {bone1Name=Dyn_cloth_02, bone2Name=Dyn_cloth_01, swing1LimitDegrees=20, swing2LimitDegrees=30, twistLimitDegrees=5}
 {bone1Name=Dyn_cloth_03, bone2Name=Dyn_cloth_02, swing1LimitDegrees=10, swing2LimitDegrees=30, twistLimitDegrees=1}
]
