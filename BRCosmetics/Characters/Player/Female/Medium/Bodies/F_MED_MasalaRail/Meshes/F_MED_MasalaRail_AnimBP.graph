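# Post-process-style animation Blueprint for the MasalaRail female character body, adding cloth and belt secondary motion to the pose copied from the attached parent mesh.
# A weighted multi-bone constraint drives Base_Cloth, then a Control Rig and pelvis-space RigidBody simulation follow.
# Additive offsets shift both belt bones only at LOD 2.
# The event graph updates LOD flags and computes gravity and rigid-body settings through a nativized parameter-asset function.
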AnimBlueprint /BRCosmetics/Characters/Player/Female/Medium/Bodies/F_MED_MasalaRail/Meshes/F_MED_MasalaRail_AnimBP
target_skeleton=&/F_MED_MasalaRail_Skeleton
graphs [
 AnimGraph {nodes [
   n1 = "Constraint - Base_Cloth (PPPPPPPP)"!(
    ComponentPose=Local To Component(LocalPose=Copy Pose From Mesh() {bUseAttachedParent=true})
    Alpha=1
    ConstraintWeights_0=0
    ConstraintWeights_1:float=0.3008
    ConstraintWeights_2:float=0.1926
    ConstraintWeights_3:float=0.2107
    ConstraintWeights_4:float=0.0658
    ConstraintWeights_5:float=0.0947
    ConstraintWeights_6:float=0.0657
    ConstraintWeights_7:float=0.0696
   ) -> (Pose) {
    boneToModify {boneName=Base_Cloth}
    constraintSetup [
     {targetBone {boneName=spine_02}}
     {targetBone {boneName=pelvis}}
     {targetBone {boneName=thigh_twist_01_l}}
     {targetBone {boneName=thigh_twist_01_r}}
     {targetBone {boneName=thigh_r}}
     {targetBone {boneName=thigh_l}}
     {targetBone {boneName=deform_groin_l}}
     {targetBone {boneName=deform_groin_r}}
    ]
    constraintWeights [1,1,1,1,1,1,1,1]
   }
   n2 = Control Rig(Source=Component To Local(ComponentPose=@n1)) {controlRigAssetReference {blueprintRigClass=&/F_MED_MasalaRail_CtrlRig.F_MED_MasalaRail_CtrlRig_C}}
   n3 = Local To Component(LocalPose=@n2)
   n4 = Get Is lod 2() -> (Is lod 2:bool)
   n5 = RigidBody!(
    ComponentPose="Transform (Modify) Bone - Bone: Base_left_belt"!(
     ComponentPose="Transform (Modify) Bone - Bone: Base_right_belt"!(ComponentPose=@n3, bAlphaBoolEnabled:bool=@n4, Translation="-2.966,0.000,0.000", Rotation="0,0,0", Scale="1,1,1") -> (Pose) {alphaInputType=Bool, translationMode=BMM_Additive, boneToModify {boneName=Base_right_belt}, translation {x=-2.966}}
     bAlphaBoolEnabled:bool=@n4
     Translation="2,0,0"
     Rotation="0,0,0"
     Scale="1,1,1"
    ) -> (Pose) {alphaInputType=Bool, translationMode=BMM_Additive, boneToModify {boneName=Base_left_belt}, translation {x=2}}
    bAlphaBoolEnabled:bool
    OverrideWorldGravity:Vector="0,0,0"
    ExternalForce
    SimSpaceSettings:Sim Space Settings Structure=(WorldAlpha=0,VelocityScaleZ=1,DampingAlpha=1,MaxLinearVelocity=10000,MaxAngularVelocity=10000,MaxLinearAcceleration=10000,MaxAngularAcceleration=10000,ExternalLinearDragV=(X=0,Y=0,Z=0),ExternalLinearVelocity=(X=0,Y=0,Z=0),ExternalAngularVelocity=(X=0,Y=0,Z=0))
   ) -> (Pose) {
    alphaInputType=Bool
    bClampLinearTranslationLimitToRefPose=true
    bFreezeIncomingPoseOnStart=true
    bOverrideWorldGravity=true
    bUseExternalClothCollision=true
    overridePhysicsAsset=&/F_MED_MasalaRail_Physics
    simulationSpace=BaseBoneSpace
    alphaBoolBlend {blendOption=Linear}
    baseBoneRef {boneName=pelvis}
    componentAppliedLinearAccClamp=<0 0 0>
    overrideWorldGravity {z=-980}
    simSpaceSettings {maxAngularAcceleration=100, maxAngularVelocity=4, maxLinearAcceleration=700, maxLinearVelocity=300, velocityScaleZ=0.5}
   }
   Output Pose(Result=Component To Local(ComponentPose=@n5)) {layerGroup=None}
  ]}
 EventGraph {nodes [
   Event BlueprintUpdateAnimation()
   > n1 = Sequence()
   > n2 = "Gravity Override State Function (Parameter Asset) (Nativized)"(
    CharacterPartAnimInstance=Self-Reference()
    GameTimeInSeconds=GetGameTimeInSeconds()
    DebugJointName=neck_02
    Parameters=&/F_MED_MasalaRail_Parameters
   )
   > Set GravityOverride(GravityOverride=@n2)
   > Set RBAN Settings Out_Native(RBAN Settings Out_Native:Fort Rigid Body Anim Node Parameters Structure=@n2.OutRigidBodyAnimNodeParameters) -> (Output_Get:Fort Rigid Body Anim Node Parameters Structure)
   n3 = Get CurrentLOD()
   Set Is lod 2(execute=@n1.then_1, Is lod 2:bool="Equal (Integer)"(A=@n3, B=2)) -> (Output_Get:bool)
   > Set is lod 3(is lod 3:bool="Equal (Integer)"(A=@n3, B=3)) -> (Output_Get:bool)
  ]}
]
variables [
 GravityOverride {type=Vector}
 Is lod 2 {type=boolean}
 is lod 3 {type=boolean}
 RBAN Settings Out_Native {type=FortRigidBodyAnimNodeParameters}
]
