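# Physics asset for the face accessory mesh of the F_MED_MasalaRail character, set up for dynamic back hair.
# A chain of three Dyn_hair_back capsules hangs from a small base body at the neck, with tight swing limits and almost no twist.
# Capsules on the upper spine, neck, clavicles and upper arms approximate the upper body, likely as hair colliders.
# Their constraints, and the hair base's link to the neck, leave swing and twist free.
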PhysicsAsset /BRCosmetics/Characters/Player/Female/Medium/Bodies/F_MED_MasalaRail/Meshes/Parts/F_MED_MasalaRail_FaceAcc_Physics
bodies {
 Base_hair_back [{shapeType=Capsule, length=0, radius=1, shapeName=Base_hair_back_capsule, rotation=<rot 90 0 0>}]
 Dyn_hair_back_01 [{shapeType=Capsule, length=6, radius=4, shapeName=Dyn_hair_back_01_capsule, center=<5.3159 0 0>, rotation=<rot 90 0 0>}]
 Dyn_hair_back_02 [{shapeType=Capsule, length=6, radius=4, shapeName=Dyn_hair_back_02_capsule, center=<5.3159 0 0>, rotation=<rot 90 0 0>}]
 Dyn_hair_back_03 [{shapeType=Capsule, length=6, radius=4, shapeName=Dyn_hair_back_03_capsule, center=<5.3159 0 0>, rotation=<rot 90 0 0>}]
 clavicle_l [{shapeType=Capsule, length=16.5, radius=4.505, shapeName=clavicle_l_capsule, center=<2.4635 0 0>, rotation=<rot 90 0 0>}]
 clavicle_r [{shapeType=Capsule, length=16.5, radius=4.505, shapeName=clavicle_r_capsule, center=<-2.4634 0 0>, rotation=<rot 90 0 0>}]
 neck_01 [{shapeType=Capsule, length=6.505, radius=4.505, shapeName=neck_01_capsule, center=<2.8594 0 -0.0093>, rotation=<rot 0.0489 -105.1558 90.1805>}]
 spine_04 [{shapeType=Capsule, length=16.5, radius=9.5, shapeName=spine_04_capsule, rotation=<rot 0 -90 0>}]
 spine_05 [{shapeType=Capsule, length=19.505, radius=6.505, shapeName=spine_05_capsule, center=<3.4048 0 0>, rotation=<rot 0 -90 0>}]
 upperarm_l [{shapeType=Capsule, length=41, radius=5.005, shapeName=upperarm_l_capsule, center=<23.5978 0 0>, rotation=<rot 90 0 0>}]
 upperarm_r [{shapeType=Capsule, length=43.5, radius=5.005, shapeName=upperarm_r_capsule, center=<-23.2944 0 0>, rotation=<rot 90 0 0>}]
}
constraints [
 {
  bone1Name=Base_hair_back
  bone2Name=neck_01
  swing1LimitDegrees=45
  swing1Motion=Free
  swing2LimitDegrees=45
  swing2Motion=Free
  twistLimitDegrees=45
  twistMotion=Free
 }
 {bone1Name=Dyn_hair_back_01, bone2Name=Base_hair_back, swing1LimitDegrees=20, swing2LimitDegrees=20, twistLimitDegrees=1}
 {bone1Name=Dyn_hair_back_02, bone2Name=Dyn_hair_back_01, swing1LimitDegrees=30, swing2LimitDegrees=20, twistLimitDegrees=1}
 {bone1Name=Dyn_hair_back_03, bone2Name=Dyn_hair_back_02, swing1LimitDegrees=20, swing2LimitDegrees=20, twistLimitDegrees=1}
 {
  bone1Name=clavicle_l
  bone2Name=spine_05
  swing1LimitDegrees=45
  swing1Motion=Free
  swing2LimitDegrees=45
  swing2Motion=Free
  twistLimitDegrees=45
  twistMotion=Free
 }
 {
  bone1Name=upperarm_l
  bone2Name=clavicle_l
  swing1LimitDegrees=45
  swing1Motion=Free
  swing2LimitDegrees=45
  swing2Motion=Free
  twistLimitDegrees=45
  twistMotion=Free
 }
 {
  bone1Name=clavicle_r
  bone2Name=spine_05
  swing1LimitDegrees=45
  swing1Motion=Free
  swing2LimitDegrees=45
  swing2Motion=Free
  twistLimitDegrees=45
  twistMotion=Free
 }
 {
  bone1Name=upperarm_r
  bone2Name=clavicle_r
  swing1LimitDegrees=45
  swing1Motion=Free
  swing2LimitDegrees=45
  swing2Motion=Free
  twistLimitDegrees=45
  twistMotion=Free
 }
 {
  bone1Name=spine_05
  bone2Name=spine_04
  swing1LimitDegrees=45
  swing1Motion=Free
  swing2LimitDegrees=45
  swing2Motion=Free
  twistLimitDegrees=45
  twistMotion=Free
 }
 {
  bone1Name=neck_01
  bone2Name=spine_05
  swing1LimitDegrees=45
  swing1Motion=Free
  swing2LimitDegrees=45
  swing2Motion=Free
  twistLimitDegrees=45
  twistMotion=Free
 }
]
