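# Physics asset for the MasalaRail female character's front face-accessory mesh, set up for dangling hair and back-accessory motion.
# Capsule chains for back, left and right hair and a back accessory hang from Base bones attached to the head.
# Chain joints have narrow swing and twist limits, with the first link of each side hair chain fully locked.
# Head, neck, upper spine, clavicle and upper-arm capsules form the body, joined by free angular constraints.
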PhysicsAsset /BRCosmetics/Characters/Player/Female/Medium/Bodies/F_MED_MasalaRail/Meshes/Parts/F_MED_MasalaRail_front_FaceAcc_Physics
bodies {
 Base_accessory_back [{shapeType=Capsule, length=0, radius=0.5, shapeName=Base_accessory_back_capsule, rotation=<rot 90 0 0>}]
 Base_hair_back [{shapeType=Capsule, length=0, radius=1, shapeName=Base_hair_back_capsule, rotation=<rot 90 0 0>}]
 Base_hair_l [{shapeType=Capsule, length=0, radius=1, shapeName=Base_hair_l_capsule, rotation=<rot 90 0 0>}]
 Base_hair_r [{shapeType=Capsule, length=0, radius=1, shapeName=Base_hair_r_capsule, rotation=<rot 90 0 0>}]
 Dyn_accessory_back_01 [{shapeType=Capsule, length=8, radius=2.5, shapeName=Dyn_accessory_back_01_capsule, center=<3.0673 0 0>, rotation=<rot 90 0 0>}]
 Dyn_accessory_back_02 [{shapeType=Capsule, length=6, radius=2, shapeName=Dyn_accessory_back_02_capsule, center=<3.0673 0 0>, rotation=<rot 90 0 0>}]
 Dyn_accessory_back_03 [{shapeType=Capsule, length=8, radius=2.5, shapeName=Dyn_accessory_back_03_capsule, center=<3.0673 0 0>, rotation=<rot 90 0 0>}]
 Dyn_hair_back_01 [{shapeType=Capsule, length=3, radius=7, shapeName=Dyn_hair_back_01_capsule, center=<5.3159 0 0>, rotation=<rot 90 0 0>}]
 Dyn_hair_back_02 [{shapeType=Capsule, length=3.5, radius=6.5, shapeName=Dyn_hair_back_02_capsule, center=<5.3159 0 0>, rotation=<rot 90 0 0>}]
 Dyn_hair_back_03 [{shapeType=Capsule, length=3, radius=7, shapeName=Dyn_hair_back_03_capsule, center=<5.3159 0 0>, rotation=<rot 90 0 0>}]
 Dyn_hair_l_01 [{shapeType=Capsule, length=6.5, radius=3.5, shapeName=Dyn_hair_l_01_capsule, center=<5.3159 0 0>, rotation=<rot 90 0 0>}]
 Dyn_hair_l_02 [{shapeType=Capsule, length=7.5, radius=2.5, shapeName=Dyn_hair_l_02_capsule, center=<5.3159 0 0>, rotation=<rot 90 0 0>}]
 Dyn_hair_l_03 [{shapeType=Capsule, length=4.5, radius=2.5, shapeName=Dyn_hair_l_03_capsule, center=<5.3159 0 0>, rotation=<rot 90 0 0>}]
 Dyn_hair_l_04 [{shapeType=Capsule, length=3.5, radius=2.5, shapeName=Dyn_hair_l_04_capsule, center=<2.1928 0 0>, rotation=<rot 90 0 0>}]
 Dyn_hair_r_01 [{shapeType=Capsule, length=6.5, radius=3.5, shapeName=Dyn_hair_r_01_capsule, center=<5.3159 0 0>, rotation=<rot 90 0 0>}]
 Dyn_hair_r_02 [{shapeType=Capsule, length=7.5, radius=2.5, shapeName=Dyn_hair_r_02_capsule, center=<5.3159 0 0>, rotation=<rot 90 0 0>}]
 Dyn_hair_r_03 [{shapeType=Capsule, length=5, radius=2, shapeName=Dyn_hair_r_03_capsule, center=<5.3159 -0.0805 0>, rotation=<rot 90 0 0>}]
 Dyn_hair_r_04 [{shapeType=Capsule, length=4, radius=2, shapeName=Dyn_hair_r_04_capsule, center=<2.1261 0.5551 0>, rotation=<rot 90 0 0>}]
 clavicle_l [{shapeType=Capsule, length=13.5, radius=4.005, shapeName=clavicle_l_capsule, center=<2.4635 0 0>, rotation=<rot 90 0 0>}]
 clavicle_r [{shapeType=Capsule, length=13.5, radius=4.005, shapeName=clavicle_r_capsule, center=<-2.4634 0 0>, rotation=<rot 90 0 0>}]
 head [{shapeType=Capsule, length=6, radius=7, shapeName=head_capsule, center=<6.1698 2.4917 0>, rotation=<rot 90 0 0>}]
 neck_01 [{shapeType=Capsule, length=7.005, radius=4.005, shapeName=neck_01_capsule, center=<2.8594 0 -0.0093>, rotation=<rot 0.0489 -105.1558 90.1805>}]
 spine_04 [{shapeType=Capsule, length=15.5, radius=10.5, shapeName=spine_04_capsule, center=<1.3059 -0.0823 0>, rotation=<rot 0 -90 0>}]
 spine_05 [{shapeType=Capsule, length=19.505, radius=6.505, shapeName=spine_05_capsule, center=<2.74 -1.4599 0>, rotation=<rot 0 -90 0>}]
 upperarm_l [{shapeType=Capsule, length=43.5, radius=2.505, shapeName=upperarm_l_capsule, center=<23.5978 0 0>, rotation=<rot 90 0 0>}]
 upperarm_r [{shapeType=Capsule, length=46, radius=2.505, shapeName=upperarm_r_capsule, center=<-23.2944 1.9814 -1.5087>, rotation=<rot 90 0 0>}]
}
constraints [
 {
  bone1Name=Base_accessory_back
  bone2Name=head
  swing1LimitDegrees=45
  swing1Motion=Free
  swing2LimitDegrees=45
  swing2Motion=Free
  twistLimitDegrees=45
  twistMotion=Free
 }
 {bone1Name=Dyn_accessory_back_01, bone2Name=Base_accessory_back, swing1LimitDegrees=20, swing2LimitDegrees=20, twistLimitDegrees=1}
 {bone1Name=Dyn_accessory_back_02, bone2Name=Dyn_accessory_back_01, swing1LimitDegrees=20, swing2LimitDegrees=20, twistLimitDegrees=1}
 {bone1Name=Dyn_accessory_back_03, bone2Name=Dyn_accessory_back_02, swing1LimitDegrees=20, swing2LimitDegrees=20, twistLimitDegrees=1}
 {
  bone1Name=Base_hair_back
  bone2Name=head
  swing1LimitDegrees=45
  swing1Motion=Free
  swing2LimitDegrees=45
  swing2Motion=Free
  twistLimitDegrees=45
  twistMotion=Free
 }
 {bone1Name=Dyn_hair_back_01, bone2Name=Base_hair_back, swing1LimitDegrees=20, swing2LimitDegrees=5, twistLimitDegrees=1}
 {bone1Name=Dyn_hair_back_02, bone2Name=Dyn_hair_back_01, swing1LimitDegrees=20, swing2LimitDegrees=10, twistLimitDegrees=1}
 {bone1Name=Dyn_hair_back_03, bone2Name=Dyn_hair_back_02, swing1LimitDegrees=20, swing2LimitDegrees=20, twistLimitDegrees=1}
 {
  bone1Name=Base_hair_l
  bone2Name=head
  swing1LimitDegrees=45
  swing1Motion=Free
  swing2LimitDegrees=45
  swing2Motion=Free
  twistLimitDegrees=45
  twistMotion=Free
 }
 {
  bone1Name=Dyn_hair_l_01
  bone2Name=Base_hair_l
  swing1LimitDegrees=1
  swing1Motion=Locked
  swing2LimitDegrees=20
  swing2Motion=Locked
  twistLimitDegrees=1
  twistMotion=Locked
 }
 {bone1Name=Dyn_hair_l_02, bone2Name=Dyn_hair_l_01, swing1LimitDegrees=5, swing2LimitDegrees=30, twistLimitDegrees=1}
 {bone1Name=Dyn_hair_l_03, bone2Name=Dyn_hair_l_02, swing1LimitDegrees=20, swing2LimitDegrees=20, twistLimitDegrees=1}
 {bone1Name=Dyn_hair_l_04, bone2Name=Dyn_hair_l_03, swing1LimitDegrees=20, swing2LimitDegrees=20, twistLimitDegrees=1}
 {
  bone1Name=Base_hair_r
  bone2Name=head
  swing1LimitDegrees=45
  swing1Motion=Free
  swing2LimitDegrees=45
  swing2Motion=Free
  twistLimitDegrees=45
  twistMotion=Free
 }
 {
  bone1Name=Dyn_hair_r_01
  bone2Name=Base_hair_r
  swing1LimitDegrees=1
  swing1Motion=Locked
  swing2LimitDegrees=20
  swing2Motion=Locked
  twistLimitDegrees=1
  twistMotion=Locked
 }
 {bone1Name=Dyn_hair_r_02, bone2Name=Dyn_hair_r_01, swing1LimitDegrees=5, swing2LimitDegrees=30, twistLimitDegrees=1}
 {bone1Name=Dyn_hair_r_03, bone2Name=Dyn_hair_r_02, swing1LimitDegrees=20, swing2LimitDegrees=20, twistLimitDegrees=1}
 {bone1Name=Dyn_hair_r_04, bone2Name=Dyn_hair_r_03, swing1LimitDegrees=20, swing2LimitDegrees=20, twistLimitDegrees=1}
 {
  bone1Name=head
  bone2Name=neck_01
  swing1LimitDegrees=45
  swing1Motion=Free
  swing2LimitDegrees=45
  swing2Motion=Free
  twistLimitDegrees=45
  twistMotion=Free
 }
 {
  bone1Name=clavicle_l
  bone2Name=spine_05
  swing1LimitDegrees=45
  swing1Motion=Free
  swing2LimitDegrees=45
  swing2Motion=Free
  twistLimitDegrees=45
  twistMotion=Free
 }
 {
  bone1Name=upperarm_l
  bone2Name=clavicle_l
  swing1LimitDegrees=45
  swing1Motion=Free
  swing2LimitDegrees=45
  swing2Motion=Free
  twistLimitDegrees=45
  twistMotion=Free
 }
 {
  bone1Name=clavicle_r
  bone2Name=spine_05
  swing1LimitDegrees=45
  swing1Motion=Free
  swing2LimitDegrees=45
  swing2Motion=Free
  twistLimitDegrees=45
  twistMotion=Free
 }
 {
  bone1Name=upperarm_r
  bone2Name=clavicle_r
  swing1LimitDegrees=45
  swing1Motion=Free
  swing2LimitDegrees=45
  swing2Motion=Free
  twistLimitDegrees=45
  twistMotion=Free
 }
 {
  bone1Name=spine_05
  bone2Name=spine_04
  swing1LimitDegrees=45
  swing1Motion=Free
  swing2LimitDegrees=45
  swing2Motion=Free
  twistLimitDegrees=45
  twistMotion=Free
 }
 {
  bone1Name=neck_01
  bone2Name=spine_05
  swing1LimitDegrees=45
  swing1Motion=Free
  swing2LimitDegrees=45
  swing2Motion=Free
  twistLimitDegrees=45
  twistMotion=Free
 }
]
