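# Post-process style Anim Blueprint for MasalaRail's face-accessory part, driving hair and accessory bones over the pose copied from the parent mesh.
# Base hair bones are constrained by front-end idle state and head roll, while head yaw rotates the back bases.
# A front RigidBody pass feeds either back physics or, with backbling worn, constraints pinning back hair.
# The event graph derives head alphas, backbling state, gravity and sim-space settings.
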AnimBlueprint /BRCosmetics/Characters/Player/Female/Medium/Bodies/F_MED_MasalaRail/Meshes/Parts/F_MED_MasalaRail_FaceAcc_AnimBP
target_skeleton=&/F_MED_MasalaRail_FaceAcc_Skeleton
graphs [
 AnimGraph {nodes [
   n1 = RigidBody!(
    ComponentPose=Local To Component(LocalPose=Use cached pose 'front hair'())
    Alpha
    OverrideWorldGravity:Vector="0,0,-800"
    ExternalForce
    SimSpaceSettings:Sim Space Settings Structure=(WorldAlpha=0,VelocityScaleZ=1,DampingAlpha=1,MaxLinearVelocity=10000,MaxAngularVelocity=10000,MaxLinearAcceleration=10000,MaxAngularAcceleration=10000,ExternalLinearDragV=(X=0,Y=0,Z=0),ExternalLinearVelocity=(X=0,Y=0,Z=0),ExternalAngularVelocity=(X=0,Y=0,Z=0))
   ) -> (Pose) {
    bClampLinearTranslationLimitToRefPose=true
    bOverrideWorldGravity=true
    overridePhysicsAsset=&/F_MED_MasalaRail_FaceAcc_Physics
    simulationSpace=BaseBoneSpace
    alphaBoolBlend {blendOption=Linear}
    baseBoneRef {boneName=neck_02}
    componentAppliedLinearAccClamp=<3000 3000 3000>
    overrideWorldGravity {z=-800}
    simSpaceSettings {maxAngularVelocity=1000, maxLinearVelocity=1000, worldAlpha=0.3}
   }
   n2 = "Constraint - Dyn_hair_back_02 (P)"(ComponentPose=Local To Component(LocalPose=Use cached pose 'front hair'()), Alpha=1, ConstraintWeights_0=1) {boneToModify {boneName=Dyn_hair_back_02}, constraintSetup [{targetBone {boneName=spine_05}}], constraintWeights [1]}
   n3 = "Constraint - Dyn_hair_back_02 (R)"(ComponentPose=@n2, Alpha=0.3, ConstraintWeights_0=1) {boneToModify {boneName=Dyn_hair_back_02}, constraintSetup [{transformType=Rotation, targetBone {boneName=head}}], constraintWeights [1]}
   n4 = "Constraint - Dyn_hair_back_03 (P)"(ComponentPose=@n3, Alpha=1, ConstraintWeights_0=1) {boneToModify {boneName=Dyn_hair_back_03}, constraintSetup [{targetBone {boneName=spine_05}}], constraintWeights [1]}
   Output Pose(Result=Blend Poses by bool(
     BlendPose_0=Component To Local(ComponentPose=@n4)
     BlendPose_1=Component To Local(ComponentPose=@n1)
     bActiveValue=Get is Wearing Backbling() -> (is Wearing Backbling:bool)
    )) {layerGroup=None}
   n5 = Get Head rot alpha() -> (Head rot alpha:double)
   n6 = "Constraint - Base_hair_l (PP)"!(
    ComponentPose=Local To Component(LocalPose=Copy Pose From Mesh() {bUseAttachedParent=true})
    Alpha=@n5
    ConstraintWeights_0=1
    ConstraintWeights_1:float=0
   ) -> (Pose) {
    alphaScaleBias {bias=1, scale=-1}
    boneToModify {boneName=Base_hair_l}
    constraintSetup [{targetBone {boneName=neck_01}}, {targetBone {boneName=clavicle_l}}]
    constraintWeights [1,1]
   }
   n7 = "Constraint - Base_hair_r (PP)"!(ComponentPose=@n6, Alpha=@n5, ConstraintWeights_0=1, ConstraintWeights_1:float=0) -> (Pose) {
    alphaScaleBias {bias=1, scale=-1}
    boneToModify {boneName=Base_hair_r}
    constraintSetup [{targetBone {boneName=neck_01}}, {targetBone {boneName=clavicle_r}}]
    constraintWeights [1,1]
   }
   n8 = Component To Local(ComponentPose=@n7)
   n9 = "Constraint - Base_hair_l (P)"!(
    ComponentPose=Local To Component(LocalPose=Copy Pose From Mesh() {bUseAttachedParent=true})
    bAlphaBoolEnabled:bool=True
    ConstraintWeights_0=1
   ) -> (Pose) {
    alphaInputType=Bool
    alphaScaleBias {bias=1, scale=-1}
    boneToModify {boneName=Base_hair_l}
    constraintSetup [{targetBone {boneName=spine_05}}]
    constraintWeights [1]
   }
   n10 = "Constraint - Base_hair_r (P)"!(ComponentPose=@n9, bAlphaBoolEnabled:bool=True, ConstraintWeights_0=1) -> (Pose) {
    alphaInputType=Bool
    alphaScaleBias {bias=1, scale=-1}
    boneToModify {boneName=Base_hair_r}
    constraintSetup [{targetBone {boneName=spine_05}}]
    constraintWeights [1]
   }
   n11 = Component To Local(ComponentPose=@n10)
   n12 = Get FE Idle() -> (FE Idle:bool)
   n13 = Get Head rot() -> (Head rot:double)
   n14 = "Transform (Modify) Bone - Bone: Base_hair_back"(
    ComponentPose=Local To Component(LocalPose=Blend Poses by bool(BlendPose_0=@n11, BlendPose_1=@n8, bActiveValue=@n12))
    Alpha=@n13
    Translation="0,0,0"
    Rotation="-0.014,-44.219,0.055"
    Scale="1,1,1"
   ) {
    rotationMode=BMM_Additive
    rotationSpace=BCS_BoneSpace
    scaleMode=BMM_Replace
    scaleSpace=BCS_BoneSpace
    translationMode=BMM_Additive
    translationSpace=BCS_BoneSpace
    boneToModify {boneName=Base_hair_back}
    rotation=<rot -0.014 -44.219 0.055>
   }
   n15 = "Transform (Modify) Bone - Bone: Base_accessory_back"(ComponentPose=@n14, Alpha=@n13, Translation="0,0,0", Rotation="-0.014,-44.219,0.055", Scale="1,1,1") {
    rotationMode=BMM_Additive
    rotationSpace=BCS_BoneSpace
    scaleMode=BMM_Replace
    scaleSpace=BCS_BoneSpace
    translationMode=BMM_Additive
    translationSpace=BCS_BoneSpace
    boneToModify {boneName=Base_accessory_back}
    rotation=<rot -0.014 -44.219 0.055>
   }
   Save cached pose 'start pose'(Pose=Component To Local(ComponentPose=@n15)) {cacheName=start pose}
   n16 = RigidBody(
    ComponentPose=Local To Component(LocalPose=Use cached pose 'start pose'())
    OverrideWorldGravity="0,0,-800"
    SimSpaceSettings=(WorldAlpha=0,VelocityScaleZ=1,DampingAlpha=1,MaxLinearVelocity=10000,MaxAngularVelocity=10000,MaxLinearAcceleration=10000,MaxAngularAcceleration=10000,ExternalLinearDragV=(X=0,Y=0,Z=0),ExternalLinearVelocity=(X=0,Y=0,Z=0),ExternalAngularVelocity=(X=0,Y=0,Z=0))
   ) {
    bClampLinearTranslationLimitToRefPose=true
    bOverrideWorldGravity=true
    overridePhysicsAsset=&/F_MED_MasalaRail_front_FaceAcc_Physics
    simulationSpace=BaseBoneSpace
    alphaBoolBlend {blendOption=Linear}
    baseBoneRef {boneName=neck_01}
    componentAppliedLinearAccClamp=<3000 3000 3000>
    overrideWorldGravity {z=-800}
    simSpaceSettings {maxAngularVelocity=1000, maxLinearVelocity=1000, worldAlpha=0.3}
   }
   Save cached pose 'front hair'(Pose=Component To Local(ComponentPose=@n16)) {cacheName=front hair}
  ]}
 EventGraph {nodes [
   n1 = Event BlueprintUpdateAnimation()
   > n2 = Sequence()
   > n3 = "Gravity Override State Function (Parameter Asset) (Nativized)"(
    CharacterPartAnimInstance=Self-Reference()
    GameTimeInSeconds=@n1.DeltaTimeX
    Parameters=&/F_MED_MasalaRail_FaceAcc_Parameters
   )
   > Set Gravity Override(Gravity Override=@n3)
   > n4 = Branch(Condition=Get bIsInFrontEnd())
   > Set Sim Space Settings(Sim Space Settings:Sim Space Settings Structure=Make Sim Space Settings!(
     WorldAlpha:float=0.1
     VelocityScaleZ:float=1
     DampingAlpha:float=1
     MaxLinearVelocity:float=800
     MaxAngularVelocity:float=8
     MaxLinearAcceleration:float=1000
     MaxAngularAcceleration:float=1000
     ExternalLinearDragV:Vector="0,0,0"
     ExternalLinearVelocity:Vector="0,0,0"
     ExternalAngularVelocity:Vector="0,0,0"
    ) -> (SimSpaceSettings:Sim Space Settings Structure) {is=title K2Node_MakeStruct}) -> (Output_Get:Sim Space Settings Structure)
   > Set Gravity Override(Gravity Override="0,0,-500")
   Set Frontend Alpha(execute=@n4.else, Frontend Alpha:double=0) -> (Output_Get:double)
   > Set Sim Space Settings(Sim Space Settings=Break Fort Rigid Body Anim Node Parameters!(FortRigidBodyAnimNodeParameters:"Fort Rigid Body Anim Node Parameters Structure (by ref)"=@n3.OutRigidBodyAnimNodeParameters) -> (ComponentLinearAccScale:Vector=~, ComponentLinearVelScale:Vector=~, SimSpaceSettings:Sim Space Settings Structure) {is=title K2Node_BreakStruct}.SimSpaceSettings)
   n5 = GetOwnerSkeletalMeshComponent()
   Set Head rot(
    execute=@n2.then_1
    Head rot:double=MapRangeClamped(
     Value=GetDeltaTransformFromRefPose!(self=@n5, BoneName=head, BaseName=spine_05) -> (
      ReturnValue_Location:Vector
      ReturnValue_Rotation_Roll:float
      ReturnValue_Rotation_Pitch:float
      ReturnValue_Rotation_Yaw:float
      ReturnValue_Scale:Vector
     ).ReturnValue_Rotation_Yaw
     InRangeA=0
     InRangeB=-50
    )
   ) -> (Output_Get:double)
   n6 = Get bIsPlayingEmote()
   > Set FE Idle(FE Idle:bool=AND Boolean(A=NOT Boolean(A=@n6), B=Get bIsInFrontEnd())) -> (Output_Get:bool)
   n7 = GetDeltaTransformFromRefPose(self=GetOwnerSkeletalMeshComponent(), BoneName=head, BaseName=spine_05)
   > n8 = Set Head Left Rot(Head Left Rot:double=MapRangeClamped(Value=@n7.ReturnValue_Rotation_Roll, InRangeA=0, InRangeB=90)) -> (Output_Get:double)
   > n9 = Set Head Right Rot(Head Right Rot:double=MapRangeClamped(Value=@n7.ReturnValue_Rotation_Roll, InRangeA=0, InRangeB=-90)) -> (Output_Get:double)
   > n10 = Branch(Condition="float > float"(A=@n9, B=@n8))
   > Set Head rot alpha(Head rot alpha:double=@n9) -> (Output_Get:double)
   Set Head rot alpha(execute=@n10.else, Head rot alpha=@n8)
   Event OnCharacterCustomizationLoaded()
   n11 = CheckForBackbling()
   > Set is Wearing Backbling(is Wearing Backbling:bool=OR Boolean!(A=@n11, B=@n11.bIsBackblingACape, C:bool=@n11.bIsBackblingLong) -> (ReturnValue)) -> (Output_Get:bool)
   Event BlueprintInitializeAnimation()
   > Set Skel_Mesh(Skel_Mesh=GetOwnerSkeletalMeshComponent())
  ]}
]
variables [
 FE Idle {type=boolean}
 Frontend Alpha {type=number}
 Gravity Override {type=Vector}
 Head Left Rot {type=number}
 Head Right Rot {type=number}
 Head rot {type=number}
 Head rot alpha {type=number}
 is Wearing Backbling {type=boolean}
 Sim Space Settings {
  type=SimSpaceSettings
  default {
   dampingAlpha=1
   maxAngularAcceleration=10000
   maxAngularVelocity=10000
   maxLinearAcceleration=10000
   maxLinearVelocity=10000
   velocityScaleZ=1
   worldAlpha=0
   externalLinearDragV=<0 0 0>
  }
 }
 Skel_Mesh {type=SkeletalMeshComponent}
]
